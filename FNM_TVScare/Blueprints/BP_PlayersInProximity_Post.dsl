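; BP_PlayersInProximity_Post is a short-lived actor that applies a transient post-process effect, named for player proximity in the TV scare feature.
; On BeginPlay it plays the Blend timeline, whose blend-in/blend-out track drives the PostProcess component's blend weight.
; The actor destroys itself when the timeline finishes.
; Overlap and Tick events are empty, so it performs no proximity check itself.
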(event EventBeginPlay
  (bind (_direction _blendinblendout) (Timeline|Blend Play)
    (:Update
      (Variables|/Script/Engine.PostProcessComponent|SetBlendWeight _blendinblendout (Variables|Default|GetPostProcess)))
    (:Finished
      (Actor|DestroyActor))))

(event Collision|EventActorBeginOverlap
  (params (OtherActor :type Actor%20Object%20Reference)))

(event EventTick
  (params (DeltaSeconds :type Float%20%28single-precision%29)))

(fn ConstructionScript)
