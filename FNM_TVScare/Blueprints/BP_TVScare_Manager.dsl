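; BP_TVScare_Manager is a haunted TV that spawns either a hostile AI or, by random roll, loot.
; The server reserves Living World AI density, and an approaching pawn switches the TV on.
; Interaction, a chance-based proximity line trace, or death before spawning triggers the outcome; spawned AI gets a stim toward the target pawn.
; RepNotifies drive client materials, music, stingers and control bus mixes; client death effects add an explosion.
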(event EventBeginPlay
  (bind _self self)
  (Networking|SwitchHasAuthority
    (:Authority
      (Class|FortAsyncActionCurrentPlaylistReady|PlaylistDataReadyAsync
        (:OnReady
          (bind _returnvalue (Class|FortAthenaLivingWorldManager|RegisterDensityReservoir (Class|FortAthenaLivingWorldManager|GetLivingWorldManager) _self 1.0 1.5e+03))
          (if _returnvalue
            (Utilities|IsValid (Variables|Default|GetPatrolPointProvider)
              (:Is%20Valid
                (Variables|InternalData|SetUserCanInteract true)
                (CallFunction|HandleShouldSpawnReward)
                ;[1 1 1] Disable Smart Objects on Spawn so that AI doesn't try to use them when the TV Breaks.
                (for _array_element (Variables|Default|GetSmartObjects)
                  (if (not (SmartObject|SetSmartObjectEnabled(defaultreason:Gameplay) (Variables|/SmartObjectLibrary/SmartObject_NoMesh_Actor.SmartObject_NoMesh_Actor_C|GetSmartObject _array_element) false :!class /Script/SmartObjectsModule.SmartObjectComponent))
                    (Fortnite|FortShippingLog self "BP_TVScare_Manager failed to disable Smart Objects." true))))
              (:Is%20Not%20Valid
                (Fortnite|FortShippingLog self "BP_TVScare_Manager doesn't have a valid Patrol Point Provider Assigned - Will not be able to spawn AI." true)))
            (else
              (Fortnite|FortShippingLog _self "BP_TVScare_Manager Tried to Register 1 AI in the Density Registry and it failed! TVScare AI will probably not be able to be spawned here!" true))))))
    (:Remote
      (Game|BindEventtoOnEndPlay (delegate Custom|CheckControlBusMixes)))))

(event Custom|CheckControlBusMixes
  (params (Actor :type Actor%20Object%20Reference) (EndPlayReason :type EEndPlayReason%20Enum :default "Destroyed"))
  (bind _cbm_proximity (Variables|Sound|GetCBM_Proximity))
  (bind _cbm_bossspawn (Variables|Sound|GetCBM_BossSpawn))
  (if (Audio|Modulation|IsControlBusMixActive _cbm_proximity)
    (Audio|Modulation|DeactivateControlBusMix _cbm_proximity)
    (else
      _))
  (if (Audio|Modulation|IsControlBusMixActive _cbm_bossspawn)
    (Audio|Modulation|DeactivateControlBusMix _cbm_bossspawn)))

(event Custom|TrySpawnAI
  (bind _returnvalue (Class|FortAthenaLivingWorldManager|GetLivingWorldManager))
  (bind _patrolpointprovider (Variables|Default|GetPatrolPointProvider))
  ;[1 1 1] Spawn AI
  (Utilities|FlowControl|DoOnce false
    (:Completed
      (if (Variables|InternalData|GetShouldSpawnRewards)
        (CallFunction|SpawnRewardInsteadOfAI)
        (else
          (Class|FortAthenaLivingWorldPointProviderInterface|EnablePointProvider_DEPRECATED _patrolpointprovider)
          (EventDispatchers|BindEventtoOnActorSpawnedResult _patrolpointprovider (delegate Custom|AISpawned))
          (Class|FortAthenaLivingWorldManager|UnregisterDensityReservoir _returnvalue self)
          (Class|FortAthenaLivingWorldManager|RequestEventGeneration _returnvalue)
          ;[1 1 1] Trigger an Ability that will apply Post Process to nearby players.
          (Ability|SendGameplayEventtoActor self (Variables|Default|GetEventTag_AIIsSpawning) (Utilities|Struct|MakeGameplayEventData (Variables|Default|GetEventTag_AIIsSpawning))))))))

(event Interaction|EventBlueprintOnInteract
  (params (InteractingPawn :type Fort%20Pawn%20Object%20Reference) (InteractionBeingAttempted :type EInteractionBeingAttempted%20Enum :default "FirstInteraction"))
  (bind _returnvalue (Class|FortAthenaLivingWorldManager|GetLivingWorldManager))
  (bind _patrolpointprovider (Variables|Default|GetPatrolPointProvider))
  ;[1 1 1] Spawn AI
  (Variables|InternalData|SetTargetPawn InteractingPawn)
  (Variables|Default|SetInteractionSuccessful true)
  (if (not (Variables|InternalData|GetShouldSpawnRewards))
    (Utilities|FlowControl|Delay (ScalableFloat|GetValueatLevel (Variables|Default|GetAnticipationDelay))))
  (Utilities|FlowControl|Sequence
    (:then_0
      (Utilities|FlowControl|DoOnce false
        (:Completed
          (if (Variables|InternalData|GetShouldSpawnRewards)
            (CallFunction|SpawnRewardInsteadOfAI)
            (else
              (Class|FortAthenaLivingWorldPointProviderInterface|EnablePointProvider_DEPRECATED _patrolpointprovider)
              (EventDispatchers|BindEventtoOnActorSpawnedResult _patrolpointprovider (delegate Custom|AISpawned))
              (Class|FortAthenaLivingWorldManager|UnregisterDensityReservoir _returnvalue self)
              (Class|FortAthenaLivingWorldManager|RequestEventGeneration _returnvalue)
              ;[1 1 1] Trigger an Ability that will apply Post Process to nearby players.
              (Ability|SendGameplayEventtoActor self (Variables|Default|GetEventTag_AIIsSpawning) (Utilities|Struct|MakeGameplayEventData (Variables|Default|GetEventTag_AIIsSpawning))))))))
    (:then_1
      (Actor|DestroyActor))))

(event Custom|AISpawned
  (params (SpawnedActor :type Actor%20Object%20Reference) (bResult :type Boolean))
  (bind _targetpawn (Variables|InternalData|GetTargetPawn))
  ;[1 1 1] Spawn AI
  (Variables|InternalData|SetAIWasSpawned true)
  (bind _asfort_athena_aibot_controller (Utilities|Casting|CastToFortAthenaAIBotController (AI|GetAIController SpawnedActor)))
  (Utilities|IsValid _targetpawn
    (:Is%20Valid
      (Class|FortAthenaAIBotController|QueueStim _asfort_athena_aibot_controller _targetpawn (Transformation|GetActorLocation _targetpawn) (Transformation|GetActorLocation _targetpawn) "Enemy" 1e+02 "None"))))

(event Custom|TriggerZinger
  ;[0.15 0.07514 0.07514 0.5] Audio
  (Variables|InternalData|GetShouldSpawnRewards
    (:then
      (CallFunction|PlaySound (Variables|Sound|GetLoot%20Burst%20Tag)))
    (:else
      (GameplayCue|ExecuteGameplayCue%28Burst%29Local (Variables|InternalData|GetTargetPawn) (Variables|Default|GetGC_Zinger) (Variables|InternalData|GetCue%20Parameters))
      (if (Pawn|IsLocallyViewed (Variables|InternalData|GetTargetPawn))
        (Audio|Modulation|ActivateControlBusMix (Variables|Sound|GetCBM_Proximity))))) ; If true, loot spawns. If false, boss spawns.
)

(event Custom|Play%20Sound
  (params (Event%20Name :type Gameplay%20Tag%20Structure))
  ;[0.15 0.07514 0.07514 0.5] Audio
  (SoundLibrary|PlaySoundSimple (WorldSubsystems|GetSoundLibrarySubsystem) self (Utilities|Struct|MakeSimpleSoundSettings Event%20Name (Transformation|GetWorldLocation (Variables|Default|GetAudio_TV_Static))) :!class /Script/SoundLibrary.SoundLibrarySubsystem))

(event Custom|PlayBossSpawnStinger
  ;[0.15 0.07514 0.07514 0.5] Audio
  (Variables|InternalData|GetShouldSpawnRewards
    (:then
      (CallFunction|PlaySound (Variables|Sound|GetLoot%20Burst%20Tag)))
    (:else
      (GameplayCue|ExecuteGameplayCue%28Burst%29Local (Variables|InternalData|GetTargetPawn) (Variables|Default|GetGC_BossSpawn) (Variables|InternalData|GetCue%20Parameters))
      (if (Pawn|IsLocallyViewed (Variables|InternalData|GetTargetPawn))
        (Audio|Modulation|ActivateControlBusMix (Variables|Sound|GetCBM_BossSpawn))))) ; If true, loot spawns. If false, boss spawns.
)

(event Custom|TriggerMatAndVO
  ;[1 1 1] Trigger TV Material Change And VO on Clients
  (Utilities|FlowControl|Sequence
    (:then_0
      (Rendering|Material|SetScalarParameterValueonMaterials (Variables|Default|GetFORT_TV) "BP_GetTime" (Utilities|Time|GetGameTimeinSeconds))
      (Rendering|Components|Light|SetIntensity (Variables|Default|GetTV_SpotLight_Green) 2e+01)
      (Rendering|Components|Light|SetIntensity (Variables|Default|GetTV_SpotLight_Blue) 0.0)
      (Components|Activation|Activate (Variables|Default|GetAudio_TV_Static) false))
    (:then_1
      (if (not (Variables|InternalData|GetCanTriggerProximityAndInteraction))
        (GameplayCue|ExecuteGameplayCue%28Burst%29Local self (Variables|Default|GetGC_InProximity) (Variables|InternalData|GetCue%20Parameters))))))

(event Custom|Stop%20TV%20Music
  (bind _returnvalue (GameplayEventRouter|GetGlobalEventRouter))
  ;[0.15 0.07514 0.07514 0.5] Audio
  (Utilities|IsValid _returnvalue
    (:Is%20Valid
      (GameplayEventRouter|ClearEventStatebyContext _returnvalue "/MusicEncounter/Events/GameplayEvent_MusicEncounter_ActiveState.GameplayEvent_MusicEncounter_ActiveState" self))))

(event Custom|Start%20TV%20Music
  (bind _returnvalue (GameplayEventRouter|GetGlobalEventRouter))
  ;[0.15 0.07514 0.07514 0.5] Audio
  (Utilities|IsValid _returnvalue
    (:Is%20Valid
      (GameplayEventRouter|BroadcastStatefulEvent _returnvalue (Utilities|Struct|MakeGameplayEventMusicEncounterActiveState "(TagName=\"\")") self))))

(event OnComponentBeginOverlap(TurnOnTVProximityTrigger)
  (params (OverlappedComponent :type Primitive%20Component%20Object%20Reference) (OtherActor :type Actor%20Object%20Reference) (OtherComp :type Primitive%20Component%20Object%20Reference) (OtherBodyIndex :type Integer) (bFromSweep :type Boolean) (SweepResult :type Hit%20Result%20Structure%20%28by%20ref%29))
  ;[1 1 1] Turn On TV when Close
  (Networking|SwitchHasAuthority
    (:Authority
      (Utilities|Casting|CastToFortPawn OtherActor)
      (Utilities|FlowControl|DoOnce false
        (:Completed
          (Components|DestroyComponent (Variables|Default|GetTurnOnTVProximityTrigger))
          (Variables|Default|SetTVState 1)
          (Utilities|FlowControl|Delay (ScalableFloat|GetValueatLevel (Variables|Default|GetVOTime))) ; Delay until VO finishes (approx)
          (Variables|InternalData|SetCanTriggerProximityAndInteraction true))))))

(event OnComponentBeginOverlap(SpawnProximityTrigger)
  (params (OverlappedComponent :type Primitive%20Component%20Object%20Reference) (OtherActor :type Actor%20Object%20Reference) (OtherComp :type Primitive%20Component%20Object%20Reference) (OtherBodyIndex :type Integer) (bFromSweep :type Boolean) (SweepResult :type Hit%20Result%20Structure%20%28by%20ref%29))
  ;[1 1 1] Pawn is too close, try to spawn the AI
  (Networking|SwitchHasAuthority
    (:Authority
      (bind _asfort_pawn (Utilities|Casting|CastToFortPawn OtherActor))
      (if (<= (Math|Random|RandomFloatinRange 0.0 1.0) (ScalableFloat|GetValueatLevel (Variables|Default|GetChanceToSpawnOnProximity)))
        (bind (_outhit _returnvalue) (Collision|LineTraceByChannel (+ (Transformation|GetActorLocation self) "0.000000,0.000000,50.000000") (Transformation|GetActorLocation _asfort_pawn) "TraceTypeQuery1" false (Utilities|Array|MakeArray self) "None" true "(R=1.000000,G=0.000000,B=0.000000,A=1.000000)" "(R=0.000000,G=1.000000,B=0.000000,A=1.000000)" 3e+01))
        (if _returnvalue
          (bind _asfort_pawn_2 (Utilities|Casting|CastToFortPawn (.HitActor (Collision|BreakHitResult _outhit))))
          (Variables|InternalData|SetSpawningByProximity true)
          (Variables|InternalData|SetTargetPawn _asfort_pawn_2)
          (if (not (Variables|InternalData|GetCanTriggerProximityAndInteraction))
            (Utilities|FlowControl|Delay 0.2))
          (Variables|Default|SetInteractionSuccessful true)
          (CallFunction|TrySpawnAI)
          (Actor|DestroyActor))
        (else
          (Components|DestroyComponent (Variables|Default|GetSpawnProximityTrigger)))))))

(event EventOnDeathServer
  (params (Damage :type Float%20%28single-precision%29) (DamageTags :type Gameplay%20Tag%20Container%20Structure%20%28by%20ref%29) (Momentum :type Vector :default "0, 0, 0") (HitInfo :type Hit%20Result%20Structure%20%28by%20ref%29) (InstigatedBy :type Controller%20Object%20Reference) (DamageCauser :type Actor%20Object%20Reference) (EffectContext :type Gameplay%20Effect%20Context%20Handle%20Structure))
  ;[1 1 1] Server Death
  (if (and (not (Variables|Default|GetInteractionSuccessful)) (not (Variables|InternalData|GetAIWasSpawned)))
    (CallFunction|TrySpawnAI)))

(event EventOnDeathPlayEffects
  (params (Damage :type Float%20%28single-precision%29) (DamageTags :type Gameplay%20Tag%20Container%20Structure%20%28by%20ref%29) (Momentum :type Vector :default "0, 0, 0") (HitInfo :type Hit%20Result%20Structure%20%28by%20ref%29) (InstigatedBy :type Fort%20Pawn%20Object%20Reference) (DamageCauser :type Actor%20Object%20Reference) (EffectContext :type Gameplay%20Effect%20Context%20Handle%20Structure))
  ;[1 1 1] Client Death
  (Niagara|SpawnSystematLocation :SystemTemplate "/FNM_TVScare/FX/NS_TVScare_Explosion.NS_TVScare_Explosion" :Location (Transformation|GetActorLocation self) :Scale "1.000000,1.000000,1.000000" :bAutoDestroy true :bAutoActivate true :PoolingMethod "None" :bPreCullCheck false)
  (SoundLibrary|PlaySoundSimple (WorldSubsystems|GetSoundLibrarySubsystem) self (Utilities|Struct|MakeSimpleSoundSettings (Variables|Sound|GetExplosion%20Tag) (Transformation|GetWorldLocation (Variables|Default|GetAudio_TV_Static))) :!class /Script/SoundLibrary.SoundLibrarySubsystem)
  (Audio|Components|Audio|FadeOut (Variables|Default|GetAudio_TV_Static) 0.2 0.0 "Linear" :!class /Script/Engine.AudioComponent))

(fn BlueprintCanInteract
  (params (InteractingPawn :type Fort%20Pawn%20Object%20Reference) (InteractionBeingAttempted :type EInteractionBeingAttempted%20Enum :default "FirstInteraction") (InteractionType :type TInteractionType%20Enum :default "IT_NoInteraction"))
  (return (and (Variables|InternalData|GetUserCanInteract) (Variables|InternalData|GetCanTriggerProximityAndInteraction))))

(fn BlueprintGetInteractionString
  (params (InteractingPawn :type Fort%20Pawn%20Object%20Reference) (InteractionBeingAttempted :type EInteractionBeingAttempted%20Enum :default "FirstInteraction"))
  (return (Variables|GetBuildingActorInteractionText)))

(fn HandleShouldSpawnReward
  (Variables|InternalData|SetShouldSpawnRewards (<= (Math|Random|RandomFloatinRange 0.0 1.0) (ScalableFloat|GetValueatLevel (Variables|Default|GetChanceToSpawnRewardInsteadOfAI))))
  (return))

(fn OnRep_InteractionSuccessful
  (Networking|SwitchHasAuthority
    (:Remote
      (if (Variables|Default|GetInteractionSuccessful)
        (CallFunction|PlayBossSpawnStinger)
        (CallFunction|StopTVMusic)))))

(fn OnRep_TargetPawn
  (Networking|SwitchHasAuthority
    (:Remote
      (if (not (Variables|InternalData|GetSpawningByProximity))
        (CallFunction|TriggerZinger)))))

(fn OnRep_TVState
  (bind _turnontvproximitytrigger (Variables|Default|GetTurnOnTVProximityTrigger))
  (Networking|SwitchHasAuthority
    (:Remote
      (Utilities|IsValid _turnontvproximitytrigger
        (:Is%20Valid
          (Components|DestroyComponent _turnontvproximitytrigger))
        (:Is%20Not%20Valid
          _))
      (switch int (Variables|Default|GetTVState)
        (:1
          (CallFunction|TriggerMatandVO))))))

(fn OnRep_UserCanInteract
  (Variables|Interaction|SetbAllowInteract true)
  (Networking|CanExecuteCosmeticEvents
    (:True
      (if (Variables|InternalData|GetUserCanInteract)
        (CallFunction|StartTVMusic)))))

(fn SpawnRewardInsteadOfAI
  (Class|AthenaFunctionLibrary|SpawnLootfromTableUsingItemEntry (Variables|Default|GetLootTierNameForReward) 1 -1 (+ (Transformation|GetActorLocation) "0.000000,0.000000,50.000000") (Math|Vector|MakeVector (Math|Random|RandomFloatinRange 0.0 1.0) (Math|Random|RandomFloatinRange 0.0 1.0) 1.0) 0 true true false 0 "Other" "NPCElimination")
  (Fortnite|FortShippingLog self "TVScare Spawned Reward instead of AI.")
  (return false))

(fn ConstructionScript)
